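; Quest gameplay ability that tracks how far the owning pawn travels while grapple-swinging and reports progress to the quest manager.
; When the Swinging tag is added, a randomized looping timer starts, accumulating whole meters between sampled locations.
; Reaching the minimum update distance sends a custom stat event and resets the total.
; Releasing the grapple fire button ends the ability, clearing the timer without a final update.
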(event Ability|EventActivateAbilityFromEvent
  (params (EventData :type Gameplay%20Event%20Data%20Structure%20%28by%20ref%29))
  (bind _owningplayerpawn (Variables|GAVars|Data|GetOwningPlayerPawn))
  ;[1 1 1] Activate ability, set up tracking vars
  (|CallParentFunction EventData)
  (Ability|Tasks|WaitGameplayTagAddToActor _owningplayerpawn "(TagName=\"MovementMode.Swinging\")"
    (:then
      (Ability|Tasks|WaitGameplayTagRemoveFromActor _owningplayerpawn "(TagName=\"Granted.Athena.GrappleParent.HoldingFireButton\")"
        (:Removed
          (Ability|EndAbility))))
    (:Added
      (bind _output_get (Variables|Default|SetPreviousLocation (Transformation|GetActorLocation (Variables|GAVars|Data|GetOwningPlayerPawn))))
      (Variables|Default|SetCurrentLocation _output_get)
      (bind _output_get_2 (Variables|Default|SetTickInterval (Math|Random|RandomFloatinRange 1.2 1.7)))
      (bind _returnvalue (Utilities|Time|SetTimerbyEvent (delegate Custom|SendDistanceUpdate) _output_get_2 true))
      (Variables|Default|SetTimerHandle _returnvalue))))

(event Custom|SendDistanceUpdate
  ;[1 1 1] Calculate and send distance update
  (Utilities|FlowControl|Sequence
    (:then_0
      (bind _output_get (Variables|Default|SetCurrentLocation (Transformation|GetActorLocation (Variables|GAVars|Data|GetOwningPlayerPawn))))
      (Variables|Default|SetMetersTravelled (+ (Math|Float|Floor (/ (Math|Vector|VectorLength (- _output_get (Variables|Default|GetPreviousLocation))) 1e+02)) (Variables|Default|GetMetersTravelled)))
      (Variables|Default|SetPreviousLocation (Variables|Default|GetCurrentLocation)))
    (:then_1
      (if (>= (Variables|Default|GetMetersTravelled) (Variables|Default|GetMinUpdateDistance))
        (Quest|SendComplexCustomStatEvent (Variables|GAVars|Data|GetOwningPlayerQuestManager) 0 (Variables|Default|GetT_Quest) "(GameplayTags=)" 1)
        (Variables|Default|SetMetersTravelled 0)))))

(event Ability|EventOnEndAbility
  (params (bWasCancelled :type Boolean))
  (bind _timerhandle (Variables|Default|GetTimerHandle))
  ;[1 1 1] On End Ability - do not send a final distance update since this calculates a single swing, not all distance
  (|CallParentFunction bWasCancelled)
  (if (Utilities|Time|IsValidTimerHandle _timerhandle)
    (Utilities|Time|ClearandInvalidateTimerbyHandle _timerhandle)))
